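# Editor widget for a single button in the tech-art Editor Widget Shelf, derived from TA_BaseEditorButtonWidget, with a label and a normally collapsed delete button.
# When list data arrives, it copies the color from EWS_Button_ListItemData into its default button color, then calls the parent handler.
# Overrides expose the border and text to the base class and toggle delete visibility.
# Clicking delete finds the parent shelf widget and asks it to remove this button.
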WidgetBlueprint /Game/Blueprints/Blutilities/ForTechArt/EditorWidgetShelf/SubWidgets/EWS_Button_Widget
graphs [
 EventGraph {nodes [
   "On Clicked (btnDelete)"!() -> (OutputDelegate, then, widget:User Widget Object Reference, event:Pointer Event Structure) {is=title K2Node_ComponentBoundEvent}
   > n1 = FindParentWidgetOfType!(
    execute
    StartingWidget:Widget Object Reference=Self-Reference()
    Type:Widget Class Reference=/Game/Blueprints/Blutilities/ForTechArt/EditorWidgetShelf/EditorWidgetShelf_EWidget.EditorWidgetShelf_EWidget_C
   ) -> (then, ReturnValue:Editor Widget Shelf EWidget Object Reference)
   > n2 = Cast To EditorWidgetShelf_EWidget!(execute, Object=@n1) -> (then, CastFailed, AsEditor Widget Shelf EWidget:Editor Widget Shelf EWidget Object Reference) {is=title K2Node_DynamicCast}
   > deleteShelfButton!(execute:Exec, self:Object Reference=@n2, button:User Widget Object Reference=Self-Reference()) -> (then:Exec, " ":bool) {is=title K2Node_Message}
   n3 = Event handleNewListDataSet!() -> (OutputDelegate, then, currentListItemData:Object Reference) {is=title K2Node_Event}
   > n4 = Cast To EWS_Button_ListItemData!(execute, Object=@n3.currentListItemData) -> (then, CastFailed, AsEWS Button List Item Data:EWS Button List Item Data Object Reference) {is=title K2Node_DynamicCast}
   > Set buttonColor_Default(buttonColor_Default:Linear Color Structure=Break EWS Button Data Struct!(EWS_ButtonData_Struct:"EWS Button Data Struct Structure (by ref)"=Get buttonData!(self:EWS Button List Item Data Object Reference=@n4) -> (buttonData:EWS Button Data Struct Structure)) -> (
     label_2_#1:String
     widgetPackageName_12_#2:String
     color_7_#3:Linear Color Structure
     consoleCommand_16_#4:String
     python_17_#5:String
     actorPackageName_22_#6:String
     actorIsTransient_24_#7:bool=~
    ) {is=title K2Node_BreakStruct}.color_7_#3) -> (Output_Get:Linear Color Structure)
   > Parent: handleNewListDataSet!(execute, currentListItemData:Object Reference=@n3.currentListItemData) -> (then) {is=title K2Node_CallParentFunction}
  ]}
 getBackgroundWidget {nodes [
   getBackgroundWidget() {is=title K2Node_FunctionEntry}
   > Return Node!(execute, backgroundWidget:Widget Object Reference=Get borderMain() -> (borderMain:Border Object Reference))
  ]}
 getLabelWidget {nodes [
   getLabelWidget() {is=title K2Node_FunctionEntry}
   > Return Node!(execute, labelWidget:Widget Object Reference=Get txtMain() -> (txtMain:Text Object Reference))
  ]}
 setDeleteVisible {nodes [
   n1 = setDeleteVisible!() -> (then, vis:bool=~) {is=title K2Node_FunctionEntry}
   > SetVisibility(
    self=Get btnDelete() -> (btnDelete:TA Generic 01 Button Object Reference)
    InVisibility=Select(Option 0=Collapsed, Option 1=Visible, Index=@n1)
   )
  ]}
 validateVisuals {nodes [
   validateVisuals() {is=title K2Node_FunctionEntry}
   > Parent: validateVisuals!(execute) -> (then, " ":bool) {is=title K2Node_CallParentFunction}
   Return Node!(execute, " ":bool=false)
  ]}
]
variables []
widgets {
 info {
  inheritedWidgetCount=0
  parentClass=/Game/Blueprints/Blutilities/ForTechArt/Widgets/Buttons/TA_BaseEditorButtonWidget.TA_BaseEditorButtonWidget_C
  rootWidgetClass=&CanvasPanel
  widgetCount=7
 }
 widgets [
  CanvasPanel_0 {widgetClassPath=&CanvasPanel, widget=CanvasPanel_0}
  borderMain {
   widgetClassPath=&Border
   bIsVariable=true
   parent=CanvasPanel_0
   widget=borderMain
   properties {padding=<ltrb 2 2 2 2>}
   slot {class=CanvasPanelSlot, properties {layoutData {alignment=<0 0>, anchors {maximum=<0 0>, minimum=<0 0>}, offsets=<ltrb 5 5 50 50>}}}
  }
  ScaleBox_0 {
   widgetClassPath=&ScaleBox
   parent=borderMain
   widget=ScaleBox_0
   properties {clipping=ClipToBounds, stretchDirection=DownOnly}
   slot {class=BorderSlot, properties {padding=<ltrb 2 2 2 2>}}
  }
  Overlay_0 {
   widgetClassPath=&Overlay
   parent=ScaleBox_0
   widget=Overlay_0
   slot {class=ScaleBoxSlot, properties {horizontalAlignment=HAlign_Fill, verticalAlignment=VAlign_Fill}}
  }
  txtMain {
   widgetClassPath=&TextBlock
   bIsVariable=true
   parent=Overlay_0
   widget=txtMain
   properties {
    clipping=ClipToBounds
    justification=Center
    text=lkfldkfldkfdldfkdlfkld
    wrapTextAt=70
    wrappingPolicy=AllowPerCharacterWrapping
    colorAndOpacity {colorUseRule=UseColor_Specified, specifiedColor=<rgba 0.0104 0.0104 0.0104 1>}
    font {
     bForceMonospaced=false
     bMaterialIsStencil=false
     fontMaterial=null
     fontObject=&Roboto
     letterSpacing=0
     monospacedWidth=1
     size=12
     skewAmount=0
     typefaceFontName=Bold
     outlineSettings {
      bApplyOutlineToDropShadows=false
      bMiteredCorners=false
      bSeparateFillAlpha=false
      outlineMaterial=null
      outlineSize=0
      outlineColor=<rgba 0 0 0 1>
     }
    }
    shadowColorAndOpacity=<rgba 0 0 0 0.2>
   }
   slot {class=OverlaySlot, properties {horizontalAlignment=HAlign_Fill, verticalAlignment=VAlign_Center}}
  }
  btnDelete {
   widgetClassPath=&TA_Generic01_Button
   parent=CanvasPanel_0
   widget=btnDelete
   properties {
    visibility=Collapsed
    buttonColor_Default=<rgba 1 0.0195 0 1>
    iconBrush {
     bIsDynamicallyLoaded=false
     drawAs=Image
     imageType=NoImage
     mirroring=NoMirror
     resourceName=None
     resourceObject=/Game/UI/Frontend/Campaign/Textures/T-Icon-ArrowDashed-64
     tiling=NoTile
     imageSize=<12 12>
     margin=<ltrb 0 0 0 0>
     outlineSettings {
      bUseBrushTransparency=false
      roundingType=HalfHeightRadius
      width=0
      color {colorUseRule=UseColor_Specified, specifiedColor=<rgba 0 0 0 0>}
      cornerRadii=<0 0 0 1>
     }
     tintColor {colorUseRule=UseColor_Specified, specifiedColor=<rgba 1 1 1 1>}
     uVRegion {bIsValid=false, max=<0 0>, min=<0 0>}
    }
   }
   slot {class=CanvasPanelSlot, properties {zOrder=1, layoutData {alignment=<0 0>, anchors {maximum=<0 0>, minimum=<0 0>}, offsets=<ltrb 0 0 20 20>}}}
  }
  SizeHolder {
   widgetClassPath=&Overlay
   parent=CanvasPanel_0
   widget=SizeHolder
   slot {class=CanvasPanelSlot, properties {layoutData {alignment=<0 0>, anchors {maximum=<0 0>, minimum=<0 0>}, offsets=<ltrb 0 0 57 60>}}}
  }
 ]
}
